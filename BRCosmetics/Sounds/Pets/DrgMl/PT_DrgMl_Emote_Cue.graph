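# Sound cue for the DrgMl pet's emote, classed under the creature pets sound class with close-range pet attenuation.
# The node graph is empty with no root, so the cue references no audio and carries only playback settings, including a raised volume multiplier.
# A concurrency override allows one instance per owner and blocks new plays, taking precedence over the assigned three-maximum concurrency set.
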SoundCue /BRCosmetics/Sounds/Pets/DrgMl/PT_DrgMl_Emote_Cue
root=null
nodes []
properties {
 attenuationSettings=&Fort_Pets_Close_Athena_Att
 bOverrideConcurrency=true
 maxDistance=10000
 soundClassObject=&Fort_Creatures_Pets
 volumeMultiplier=1.568
 concurrencyOverrides {bLimitToOwner=true, maxCount=1, resolutionRule=PreventNew}
 concurrencySet [/Game/Sounds/Concurrency_Rules/ThreeMaximum_Owner_PreventNew]
}
