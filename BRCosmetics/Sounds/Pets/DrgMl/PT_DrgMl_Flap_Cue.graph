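# Sound cue for the wing flap of the DrgMl cosmetic pet, played with close pet attenuation in the creature pets sound class.
# A quality-level node sends its first two slots through a modulator to a random pick of two flap waves.
# The last two slots use a modulator on a single "_opt" wave.
# A concurrency override allows one instance per owner, stopping the oldest, despite an assigned shared concurrency set.
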SoundCue /BRCosmetics/Sounds/Pets/DrgMl/PT_DrgMl_Flap_Cue
root=QualityLevel_0
nodes [
 {id=Modulator_0, children [Random_2]}
 {id=Modulator_1, children [WavePlayer_1]}
 {id=QualityLevel_0, children [Modulator_0,Modulator_0,Modulator_1,Modulator_1]}
 {id=Random_2, children [WavePlayer_0,WavePlayer_3], properties {weights [1,1]}}
 {id=WavePlayer_0, properties {soundWaveAssetPtr=&/PT_DrgMl_Flap_01}}
 {id=WavePlayer_1, properties {soundWaveAssetPtr=&/PT_DrgMl_Flap_01_opt}}
 {id=WavePlayer_3, properties {soundWaveAssetPtr=&/PT_DrgMl_Flap_02}}
]
properties {
 attenuationSettings=&Fort_Pets_Close_Athena_Att
 bOverrideConcurrency=true
 maxDistance=1800.0001
 soundClassObject=&Fort_Creatures_Pets
 volumeMultiplier=1.68
 concurrencyOverrides {bLimitToOwner=true, maxCount=1, resolutionRule=StopOldest}
 concurrencySet [/Game/Sounds/Concurrency_Rules/ThreeMaximum_Owner_PreventNew]
}
