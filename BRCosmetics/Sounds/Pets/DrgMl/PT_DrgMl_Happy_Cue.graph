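# Sound cue for the happy sound of the DrgMl pet, using the creature pets sound class and close-range pet attenuation.
# A quality level node sends its first two levels to an equally weighted random pick of three waves, and its last two to one "_opt" wave.
# An owner-limited concurrency override caps simultaneous plays at two and blocks new ones, superseding the assigned three-maximum concurrency set.
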SoundCue /BRCosmetics/Sounds/Pets/DrgMl/PT_DrgMl_Happy_Cue
root=QualityLevel_0
nodes [
 {id=QualityLevel_0, children [Random_0,Random_0,WavePlayer_0,WavePlayer_0]}
 {id=Random_0, children [WavePlayer_1,WavePlayer_2,WavePlayer_3], properties {weights [1,1,1]}}
 {id=WavePlayer_0, properties {soundWaveAssetPtr=&/PT_DrgMl_Happy_01_opt}}
 {id=WavePlayer_1, properties {soundWaveAssetPtr=&/PT_DrgMl_Happy_01}}
 {id=WavePlayer_2, properties {soundWaveAssetPtr=&/PT_DrgMl_Happy_02}}
 {id=WavePlayer_3, properties {soundWaveAssetPtr=&/PT_DrgMl_Happy_03}}
]
properties {
 attenuationSettings=&Fort_Pets_Close_Athena_Att
 bOverrideConcurrency=true
 maxDistance=1800.0001
 soundClassObject=&Fort_Creatures_Pets
 volumeMultiplier=1.12
 concurrencyOverrides {bLimitToOwner=true, maxCount=2, resolutionRule=PreventNew}
 concurrencySet [/Game/Sounds/Concurrency_Rules/ThreeMaximum_Owner_PreventNew]
}
