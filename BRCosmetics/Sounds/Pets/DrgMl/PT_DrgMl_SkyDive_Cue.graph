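# Sound cue for the DrgMl pet's skydive sound, in the creature pets sound class with close-range pet attenuation.
# A quality-level switch sends two levels to an equally weighted random pick of three waves and the other two to a single optimized wave.
# Each branch passes through a modulator for variation.
# A concurrency override allows one instance per owner and prevents new plays while it is active.
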SoundCue /BRCosmetics/Sounds/Pets/DrgMl/PT_DrgMl_SkyDive_Cue
root=QualityLevel_0
nodes [
 {id=Modulator_0, children [Random_1], properties {pitchMax=1}}
 {id=Modulator_1, children [WavePlayer_1], properties {pitchMax=1}}
 {id=QualityLevel_0, children [Modulator_0,Modulator_0,Modulator_1,Modulator_1]}
 {id=Random_1, children [WavePlayer_0,WavePlayer_3,WavePlayer_5], properties {weights [1,1,1]}}
 {id=WavePlayer_0, properties {soundWaveAssetPtr=&/PT_DrgMl_SkyDive_01}}
 {id=WavePlayer_1, properties {soundWaveAssetPtr=&/PT_DrgMl_SkyDive_01_opt}}
 {id=WavePlayer_3, properties {soundWaveAssetPtr=&/PT_DrgMl_SkyDive_02}}
 {id=WavePlayer_5, properties {soundWaveAssetPtr=&/PT_DrgMl_SkyDive_03}}
]
properties {
 attenuationSettings=&Fort_Pets_Close_Athena_Att
 bOverrideConcurrency=true
 maxDistance=1800.0001
 soundClassObject=&Fort_Creatures_Pets
 volumeMultiplier=1.12
 concurrencyOverrides {bLimitToOwner=true, maxCount=1, resolutionRule=PreventNew}
 concurrencySet [/Game/Sounds/Concurrency_Rules/ThreeMaximum_Owner_PreventNew]
}
